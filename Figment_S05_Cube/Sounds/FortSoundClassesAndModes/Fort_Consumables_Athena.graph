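# Sound class for Athena (Battle Royale) consumable sound effects in the Figment S05 Cube content.
# It is a child of the Fort_Athena_SFX class, placing consumable sounds under that SFX group in the mix hierarchy.
# Its own properties lower the volume slightly and send a partial low-frequency bleed to the LFE channel.
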SoundClass /Figment_S05_Cube/Sounds/FortSoundClassesAndModes/Fort_Consumables_Athena
parent=&/Fort_Athena_SFX
properties {properties {lFEBleed=0.25, volume=0.8}}
