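# Sound class for searching reward sounds in the Figment_S05_Cube content, parented to the player searching, looting and crafting class.
# It carries a passive sound mix modifier that applies the weapon duck mix while its sounds play above a minimum volume threshold.
# The class also sets an LFE bleed amount for its sounds.
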SoundClass /Figment_S05_Cube/Sounds/FortSoundClassesAndModes/Fort_Searching_Reward_PSM
parent=&/Fort_08_Player_Searching_looting_crafting
properties {
 passiveSoundMixModifiers [{minVolumeThreshold=0.3, soundMix=&/Fort_Weapon_Duck_MixMod}]
 properties {lFEBleed=0.25}
}
