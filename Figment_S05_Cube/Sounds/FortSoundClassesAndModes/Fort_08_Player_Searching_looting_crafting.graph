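# Sound class that groups player searching, looting and crafting audio under the SFX parent class.
# It parents six child classes covering doors, loot collection, impacts, searching reward, high-priority player loot and loot drop/collect suppression, with some names suggesting passive sound mix targets.
# Its own properties set a partial LFE bleed.
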SoundClass /Figment_S05_Cube/Sounds/FortSoundClassesAndModes/Fort_08_Player_Searching_looting_crafting
parent=&SFX
children [
 &Fort_Doors
 &Fort_Loot_Collect
 &Fort_Impacts
 &/Fort_Searching_Reward_PSM
 &/Fort_Player_Loot_High_PSM
 &/Fort_Loot_Drop_Collect_Suppress
]
properties {properties {lFEBleed=0.25}}
